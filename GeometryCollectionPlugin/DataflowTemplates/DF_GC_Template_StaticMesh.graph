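# Dataflow template that turns a static mesh into a fractured Geometry Collection for destruction.
# A SourceStaticMesh variable, defaulting to the engine cube, feeds StaticMeshToCollection, which sets internal faces from material index.
# The collection passes through a uniform Voronoi fracture with a fixed site count and no noise or grout.
# The terminal node receives the fractured collection, plus materials, instanced meshes and root proxy meshes straight from the conversion node.
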Dataflow /GeometryCollectionPlugin/DataflowTemplates/DF_GC_Template_StaticMesh
name=DF_GC_Template_StaticMesh
connections [
 {
  fromNode=SourceStaticMesh
  fromPin=Value
  fromPinGuid=#1
  toNode=StaticMeshToCollection
  toPin=StaticMesh
  toPinGuid=#2
 }
 {
  fromNode=StaticMeshToCollection
  fromPin=Collection
  fromPinGuid=#3
  toNode=UniformFracture
  toPin=Collection
  toPinGuid=#4
 }
 {
  fromNode=UniformFracture
  fromPin=Collection
  fromPinGuid=#5
  toNode=GeometryCollectionTerminal
  toPin=Collection
  toPinGuid=#6
 }
 {
  fromNode=StaticMeshToCollection
  fromPin=Materials
  fromPinGuid=#7
  toNode=GeometryCollectionTerminal
  toPin=Materials
  toPinGuid=#8
 }
 {
  fromNode=StaticMeshToCollection
  fromPin=InstancedMeshes
  fromPinGuid=#9
  toNode=GeometryCollectionTerminal
  toPin=InstancedMeshes
  toPinGuid=#10
 }
 {
  fromNode=StaticMeshToCollection
  fromPin=RootProxyMeshes
  fromPinGuid=#11
  toNode=GeometryCollectionTerminal
  toPin=RootProxyMeshes
  toPinGuid=#12
 }
]
nodes [
 SourceStaticMesh {
  type=FGetDataflowVariableNode
  guid=#13
  outputPins [Value {type="TObjectPtr<UStaticMesh>"}]
  properties {
   VariableName=SourceStaticMesh
   VariablePropertyBag=(Value=/Engine/Transient.PropertyBag_23b9fa6cfb5bbaf5(SourceStaticMesh="&StaticMesh'/Engine/BasicShapes/Cube'"))
  }
 }
 StaticMeshToCollection {
  type=FStaticMeshToCollectionDataflowNode_v2
  guid=#14
  inputPins [
   StaticMesh {type="TObjectPtr<UStaticMesh>"}
   MeshTransform {type=FTransform}
  ]
  outputPins [
   Collection {type=FManagedArrayCollection}
   Materials {type="TArray<TObjectPtr<UMaterialInterface>>"}
   InstancedMeshes {type="TArray<FGeometryCollectionAutoInstanceMesh>"}
   RootProxyMeshes {type="TArray<FDataflowRootProxyMesh>"}
  ]
  properties {
   MeshTransform=(Rotation=(X=0,Y=0,Z=0,W=1),Translation=(X=0,Y=0,Z=0),Scale3D=(X=1,Y=1,Z=1))
   StaticMesh=None
   bSetInternalFromMaterialIndex=True
   bSplitComponents=False
  }
 }
 UniformFracture {
  type=FUniformFractureDataflowNode
  guid=#15
  inputPins [
   Collection {type=FManagedArrayCollection}
   TransformSelection {type=FDataflowTransformSelection}
   Transform {type=FTransform}
   MinVoronoiSites {type=int32}
   MaxVoronoiSites {type=int32}
   InternalMaterialID {type=int32}
   RandomSeed {type=int32}
   ChanceToFracture {type=float}
   Grout {type=float}
   Amplitude {type=float}
   Frequency {type=float}
   Persistence {type=float}
   Lacunarity {type=float}
   OctaveNumber {type=int32}
   PointSpacing {type=float}
   CollisionSampleSpacing {type=float}
  ]
  outputPins [
   Collection {type=FManagedArrayCollection}
   TransformSelection {type=FDataflowTransformSelection}
   NewGeometryTransformSelection {type=FDataflowTransformSelection}
  ]
  properties {
   AddSamplesForCollision=False
   Amplitude=0
   ChanceToFracture=1
   CloseVertexDistance=0.001
   CollisionSampleSpacing=50
   Frequency=0.1
   GroupFracture=True
   Grout=0
   InternalMaterialID=0
   Lacunarity=2
   MaxVoronoiSites=20
   MinVoronoiSites=20
   OctaveNumber=4
   Persistence=0.5
   PointSpacing=10
   RandomSeed=0
   SplitIslands=True
   VertexToSurfaceBridgeDistance=0
  }
 }
 GeometryCollectionTerminal {
  type=FGeometryCollectionTerminalDataflowNode_v2
  guid=#16
  inputPins [
   Collection {type=FManagedArrayCollection}
   Materials {type="TArray<TObjectPtr<UMaterialInterface>>"}
   InstancedMeshes {type="TArray<FGeometryCollectionAutoInstanceMesh>"}
   RootProxyMeshes {type="TArray<FDataflowRootProxyMesh>"}
  ]
  outputPins [
   Collection {type=FManagedArrayCollection}
   Materials {type="TArray<TObjectPtr<UMaterialInterface>>"}
   InstancedMeshes {type="TArray<FGeometryCollectionAutoInstanceMesh>"}
  ]
 }
]
